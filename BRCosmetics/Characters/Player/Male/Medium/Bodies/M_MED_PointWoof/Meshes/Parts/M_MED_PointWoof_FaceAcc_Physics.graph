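# Physics asset for the PointWoof male medium character's FaceAcc mesh part, providing capsule bodies for a dangling tie and the upper body around it.
# Five dyn_tie bodies form a chain from spine_05, with a free root joint and swing-limited, zero-twist-limit links below it.
# Spine, arm and hand capsules cover the body, with free spine joints, narrow arm limits and locked wrists.
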PhysicsAsset /BRCosmetics/Characters/Player/Male/Medium/Bodies/M_MED_PointWoof/Meshes/Parts/M_MED_PointWoof_FaceAcc_Physics
bodies {
 dyn_tie_01 [{shapeType=Capsule, length=5.005, radius=1.005, shapeName=dyn_tie_01_capsule}]
 dyn_tie_02 [{shapeType=Capsule, length=1.6077, radius=1.6394, shapeName=dyn_tie_02_capsule, center=<1.9214 -1.0494 0>}]
 dyn_tie_03 [{shapeType=Capsule, length=4.3359, radius=2, shapeName=dyn_tie_03_capsule, center=<3.4725 -0.7382 0>, rotation=<rot 90 0 0>}]
 dyn_tie_04 [{shapeType=Capsule, length=4.505, radius=2, shapeName=dyn_tie_04_capsule, center=<5.4992 -1.2314 0>, rotation=<rot 90 -178.2101 -174.2547>}]
 dyn_tie_05 [{shapeType=Capsule, length=6.0401, radius=2.288, shapeName=dyn_tie_05_capsule, center=<4.382 -1.9412 0>, rotation=<rot 90 180 -177.6076>}]
 hand_l [{shapeType=Capsule, length=1, radius=9, shapeName=hand_l_capsule, center=<5.856 0 0>}]
 hand_r [{shapeType=Capsule, length=1, radius=9, shapeName=hand_r_capsule, center=<-8.0505 0 0>, rotation=<rot 0 -90 0>}]
 lowerarm_l [{shapeType=Capsule, length=17, radius=4.005, shapeName=lowerarm_l_capsule, center=<10.7055 0 0>, rotation=<rot 90 0 0>}]
 lowerarm_r [{shapeType=Capsule, length=16, radius=5.005, shapeName=lowerarm_r_capsule, center=<-10.3796 0 0>, rotation=<rot 90 0 0>}]
 spine_02 [{shapeType=Capsule, length=0.5, radius=13.005, shapeName=spine_02_capsule, center=<2.8545 -2.083 0>, rotation=<rot 0 -90 0>}]
 spine_03 [{shapeType=Capsule, length=10.005, radius=13.005, shapeName=spine_03_capsule, center=<4.8561 -1.7765 0>, rotation=<rot 0 -90 0>}]
 spine_04 [{shapeType=Capsule, length=15.505, radius=10.505, shapeName=spine_04_capsule, center=<3.1195 0.901 0>, rotation=<rot 0 -90 0>}]
 spine_05 [{shapeType=Capsule, length=13.505, radius=10.005, shapeName=spine_05_capsule, center=<-0.7993 0.68 0>, rotation=<rot 0 -90 0>}]
 upperarm_l [{shapeType=Capsule, length=16, radius=5.005, shapeName=upperarm_l_capsule, center=<11.1072 0 0>, rotation=<rot 90 0 0>}]
 upperarm_r [{shapeType=Capsule, length=16, radius=5.005, shapeName=upperarm_r_capsule, center=<-10.43 0 0>, rotation=<rot 90 0 0>}]
}
constraints [
 {
  bone1Name=spine_05
  bone2Name=spine_04
  swing1LimitDegrees=45
  swing1Motion=Free
  swing2LimitDegrees=45
  swing2Motion=Free
  twistLimitDegrees=45
  twistMotion=Free
 }
 {
  bone1Name=spine_04
  bone2Name=spine_03
  swing1LimitDegrees=45
  swing1Motion=Free
  swing2LimitDegrees=45
  swing2Motion=Free
  twistLimitDegrees=45
  twistMotion=Free
 }
 {bone1Name=dyn_tie_05, bone2Name=dyn_tie_04, swing1LimitDegrees=40, swing2LimitDegrees=15, twistLimitDegrees=0}
 {
  bone1Name=dyn_tie_01
  bone2Name=spine_05
  swing1LimitDegrees=45
  swing1Motion=Free
  swing2LimitDegrees=45
  swing2Motion=Free
  twistLimitDegrees=45
  twistMotion=Free
 }
 {bone1Name=dyn_tie_02, bone2Name=dyn_tie_01, swing1LimitDegrees=40, swing2LimitDegrees=15, twistLimitDegrees=0}
 {bone1Name=dyn_tie_03, bone2Name=dyn_tie_02, swing1LimitDegrees=40, swing2LimitDegrees=15, twistLimitDegrees=0}
 {bone1Name=dyn_tie_04, bone2Name=dyn_tie_03, swing1LimitDegrees=40, swing2LimitDegrees=30, twistLimitDegrees=0}
 {
  bone1Name=spine_03
  bone2Name=spine_02
  swing1LimitDegrees=45
  swing1Motion=Free
  swing2LimitDegrees=45
  swing2Motion=Free
  twistLimitDegrees=45
  twistMotion=Free
 }
 {bone1Name=upperarm_l, bone2Name=spine_05, swing1LimitDegrees=25, swing2LimitDegrees=25, twistLimitDegrees=0}
 {bone1Name=lowerarm_l, bone2Name=upperarm_l, swing1LimitDegrees=25, swing2LimitDegrees=25, twistLimitDegrees=0}
 {bone1Name=upperarm_r, bone2Name=spine_05, swing1LimitDegrees=25, swing2LimitDegrees=25, twistLimitDegrees=0}
 {bone1Name=lowerarm_r, bone2Name=upperarm_r, swing1LimitDegrees=25, swing2LimitDegrees=25, twistLimitDegrees=0}
 {
  bone1Name=hand_l
  bone2Name=lowerarm_l
  swing1LimitDegrees=45
  swing1Motion=Locked
  swing2LimitDegrees=45
  swing2Motion=Locked
  twistLimitDegrees=45
  twistMotion=Locked
 }
 {
  bone1Name=hand_r
  bone2Name=lowerarm_r
  swing1LimitDegrees=45
  swing1Motion=Locked
  swing2LimitDegrees=45
  swing2Motion=Locked
  twistLimitDegrees=45
  twistMotion=Locked
 }
]
