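# Physics asset for the M_MED_PointWoof male character's chain accessory part, giving capsule bodies to its dangling belt bones.
# Front and side chains lock their first links to base bodies, allow limited swing further down, and join at their tips through an angularly free constraint, forming a loop.
# A separate two-body "waych" pendant has limited swing and twist.
# Left groin and thigh helper capsules are unconstrained, apparently serving as leg colliders.
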PhysicsAsset /BRCosmetics/Characters/Player/Male/Medium/Bodies/M_MED_PointWoof/Meshes/Parts/M_MED_PointWoof_FaceAcc_chain_Physics
bodies {
 base_belt_fr_00 [{shapeType=Capsule, length=0.505, radius=0.505, shapeName=base_belt_fr_00_capsule, rotation=<rot 0 -90 0>}]
 base_belt_side_00 [{shapeType=Capsule, length=0.505, radius=0.505, shapeName=base_belt_side_00_capsule, rotation=<rot 0 -90 0>}]
 deform_groin_l [{shapeType=Capsule, length=8.5, radius=3.905, shapeName=thigh_l_capsule, center=<-0.3124 -18.9807 10.234>, rotation=<rot -5.2179 0 25.591>}]
 dyn_belt_fr_01 [{shapeType=Capsule, length=0.505, radius=0.505, shapeName=dyn_belt_fr_01_capsule, rotation=<rot 90 0 0>}]
 dyn_belt_fr_02 [{
   shapeType=Capsule
   length=4.2219
   radius=1.5
   shapeName=dyn_belt_fr_02_capsule
   center=<2.8949 -0.0079 -0.3153>
   rotation=<rot -78.7153 -8.8602 8.8452>
  }]
 dyn_belt_fr_03 [{shapeType=Capsule, length=2.8487, radius=0.9, shapeName=dyn_belt_fr_03_capsule, center=<1.4156 0 0>, rotation=<rot 90 0 0>}]
 dyn_belt_side_01 [{shapeType=Capsule, length=1.5, radius=1.005, shapeName=dyn_belt_side_01_capsule, center=<2.06 0 0>, rotation=<rot 90 0 0>}]
 dyn_belt_side_02 [{shapeType=Capsule, length=4.5533, radius=1, shapeName=dyn_belt_side_02_capsule, center=<2.0359 0 0>, rotation=<rot 90 0 0>}]
 dyn_belt_side_03 [{shapeType=Capsule, length=2.919, radius=0.9, shapeName=dyn_belt_side_03_capsule, center=<1.6115 0 -0.0548>, rotation=<rot 85 0.0026 0.0024>}]
 dyn_belt_waych_01 [{
   shapeType=Capsule
   length=0.7129
   radius=0.671
   shapeName=dyn_belt_waych_01_capsule
   center=<0.5185 0.0017 -0.0624>
   rotation=<rot -2.373 -90 -0.5081>
  }]
 dyn_belt_waych_02 [{shapeType=Capsule, length=2.6915, radius=2.3473, shapeName=dyn_belt_waych_02_capsule, center=<3.801 -1.43 0>, rotation=<rot 0 -90 0>}]
 helper_thigh_l [{shapeType=Capsule, length=10.7, radius=7.505, shapeName=thigh_l_capsule, center=<-7.1372 1.002 0.6485>, rotation=<rot 90 0 0>}]
}
constraints [
 {bone1Name=dyn_belt_waych_02, bone2Name=dyn_belt_waych_01, swing1LimitDegrees=40, swing2LimitDegrees=40, twistLimitDegrees=15}
 {
  bone1Name=dyn_belt_fr_01
  bone2Name=base_belt_fr_00
  swing1LimitDegrees=20
  swing1Motion=Locked
  swing2LimitDegrees=20
  swing2Motion=Locked
  twistLimitDegrees=0
  twistMotion=Locked
 }
 {
  bone1Name=dyn_belt_side_01
  bone2Name=base_belt_side_00
  swing1LimitDegrees=20
  swing1Motion=Locked
  swing2LimitDegrees=20
  swing2Motion=Locked
  twistLimitDegrees=0
  twistMotion=Locked
 }
 {bone1Name=dyn_belt_fr_03, bone2Name=dyn_belt_fr_02, swing1LimitDegrees=10, swing2LimitDegrees=32, twistLimitDegrees=0}
 {bone1Name=dyn_belt_fr_02, bone2Name=dyn_belt_fr_01, swing1LimitDegrees=10, swing2LimitDegrees=20, twistLimitDegrees=0}
 {bone1Name=dyn_belt_side_03, bone2Name=dyn_belt_side_02, swing1LimitDegrees=10, swing2LimitDegrees=32, twistLimitDegrees=0}
 {bone1Name=dyn_belt_side_02, bone2Name=dyn_belt_side_01, swing1LimitDegrees=20, swing2LimitDegrees=25, twistLimitDegrees=0}
 {
  bone1Name=dyn_belt_fr_03
  bone2Name=dyn_belt_side_03
  swing1LimitDegrees=45
  swing1Motion=Free
  swing2LimitDegrees=45
  swing2Motion=Free
  twistLimitDegrees=45
  twistMotion=Free
 }
]
